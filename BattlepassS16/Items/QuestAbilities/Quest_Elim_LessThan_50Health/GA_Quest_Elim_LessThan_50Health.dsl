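; Battle pass Season 16 quest gameplay ability that marks the owning player while their health is at or below 50, for a low-health elimination quest.
; On activation it binds a health check to the pawn's health-changed event and runs it once immediately.
; The check applies the quest GameplayEffect to self at or below the threshold, otherwise removes active effects granting the Athena.Quests.Elim.Lessthan50health tag.
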(event Ability|EventActivateAbility
  (|CallParentFunction)
  (Health|BindEventtoOnPawnHealthChanged (Variables|GAVars|Data|GetOwningPlayerPawn) (delegate Custom|Check%20Health))
  (CallFunction|CheckHealth))

(event Ability|EventFailedToActivatePassiveAbility
  (params (ActorInfo :type Gameplay%20Ability%20Actor%20Info%20Structure)))

(event Custom|Check%20Health
  (if (<= (FortPawn|GetHealth (Variables|GAVars|Data|GetOwningPlayerPawn)) 5e+01)
    (GameplayEffects|ApplyGameplayEffectToSelf :self (Ability|GetAbilitySystemComponent (Variables|GAVars|Data|GetOwningPlayerPawn)) :GameplayEffectClass "/BattlepassS16/Items/QuestAbilities/Quest_Elim_LessThan_50Health/GE_Quest_Elim_LessThan_50Health.GE_Quest_Elim_LessThan_50Health_C" :Level 0.0)
    (else
      (GameplayEffects|RemoveActiveEffectswithGrantedTags (Ability|GetAbilitySystemComponent (Variables|GAVars|Data|GetOwningPlayerPawn)) "(GameplayTags=((TagName=\"Athena.Quests.Elim.Lessthan50health\")))"))))
